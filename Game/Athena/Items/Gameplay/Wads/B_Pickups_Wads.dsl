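; Pickup visual for Wads, the bar item, that scales its look and sound to the stack size.
; On attach, cosmetic-only logic reads the parent FortPickup's count, compares it to ScalableFloat thresholds, and sets a small, medium or large mesh, glow and drop sound.
; Pickup sound, also size-based, plays only when player pawn 0 collects it, then attaches to the collector as it enters the backpack.
; A debug PrintString remains on attach.
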(event EventBeginPlay
  (|CallParentFunction))

(event EventOnTossed
  (|CallParentFunction))

(event EventOnAttached
  (bind _returnvalue (Actor|GetAttachParentActor))
  (bind _staticmeshcomponent (Variables|FortPickupsParent|GetStaticMeshComponent))
  (bind _staticmeshcomponent_2 (Variables|FortPickupsParent|GetStaticMeshComponent))
  (bind _staticmeshcomponent_3 (Variables|FortPickupsParent|GetStaticMeshComponent))
  (|CallParentFunction)
  (Networking|CanExecuteCosmeticEvents
    (:True
      (bind _asfort_pickup (Utilities|Casting|CastToFortPickup _returnvalue))
      (Variables|Default|SetWadCount (.Count (Class|FortItemFunctionLibrary|BreakFortItemEntry (Pickup|GetItemEntry _asfort_pickup 0))))
      (Development|PrintString (Utilities|String|Append "Attach Parent Actor" (Utilities|GetDisplayName _returnvalue) "   stack count:" (Utilities|String|ToString(Integer) (.Count (Class|FortItemFunctionLibrary|BreakFortItemEntry (Pickup|GetItemEntry _asfort_pickup 0))))))
      (CallFunction|PlayDropSound)
      ;[1 1 1] Large
      (if (>= (Variables|Default|GetWadCount) (Math|Float|Truncate (ScalableFloat|GetValueatLevel (Variables|Default|GetLargeThreshold))))
        (Components|StaticMesh|SetStaticMesh _staticmeshcomponent_2 "/Game/Athena/Items/Gameplay/Wads/Item_Bars/Meshes/Item_Bars_L.Item_Bars_L")
        (Transformation|SetWorldScale3D _staticmeshcomponent_2 (Math|Conversions|ToVector(Float) (Variables|Default|GetScale)))
        (CallFunction|SetupPickupGlow 2)
        (elif (>= (Variables|Default|GetWadCount) (Math|Float|Truncate (ScalableFloat|GetValueatLevel (Variables|Default|GetMediumThreshold))))
          ;[1 1 1] Med
          (Components|StaticMesh|SetStaticMesh _staticmeshcomponent "/Game/Athena/Items/Gameplay/Wads/Item_Bars/Meshes/Item_Bars_M.Item_Bars_M")
          (Transformation|SetWorldScale3D _staticmeshcomponent (Math|Conversions|ToVector(Float) (Variables|Default|GetScale)))
          (CallFunction|SetupPickupGlow 1)
          (else
            ;[1 1 1] Small
            (Components|StaticMesh|SetStaticMesh _staticmeshcomponent_3 "/Game/Athena/Items/Gameplay/Wads/Item_Bars/Meshes/Item_Bars_S.Item_Bars_S")
            (Transformation|SetWorldScale3D _staticmeshcomponent_3 (Math|Conversions|ToVector(Float) (Variables|Default|GetScale)))
            (CallFunction|SetupPickupGlow 0)))))))

(event EventOnPickedUp
  (params (PickupTarget :type Fort%20Pawn%20Object%20Reference))
  (|CallParentFunction 0)
  (CallFunction|PlayPickUpSound PickupTarget))

(event Custom|Play%20PickUp%20Sound
  (params (Pickup%20Target :type Fort%20Pawn%20Object%20Reference))
  (if (== Pickup%20Target (Game|GetPlayerPawn 0))
    (bind _returnvalue (Audio|SpawnSoundatLocation (Utilities|Select :Index (CallFunction|GetWadSizeValue) :Option%200 (Variables|Sound|GetSmallWadPickUpSound) :Option%201 (Variables|Sound|GetMediumWadPickUpSound) :Option%202 (Variables|Sound|GetLargeWadPickUpSound)) (Transformation|GetActorLocation)))
    (Variables|Default|SetPickUpSound _returnvalue)))

(event Custom|Play%20Drop%20Sound
  (Audio|Components|Audio|SetSound (Variables|Default|GetDropSound) (Utilities|Select :Index (CallFunction|GetWadSizeValue) :Option%200 (Variables|Sound|GetSmallWadDropSound) :Option%201 (Variables|Sound|GetMediumWadDropSound) :Option%202 (Variables|Sound|GetLargeWadDropSound)))
  (Audio|Components|Audio|Play (Variables|Default|GetDropSound) 0.0))

(event EventOnAboutToEnterBackpack
  (params (PickupTarget :type Fort%20Pawn%20Object%20Reference))
  (bind _pickupsound (Variables|Default|GetPickUpSound))
  (Utilities|IsValid PickupTarget
    (:Is%20Valid
      (Utilities|IsValid _pickupsound
        (:Is%20Valid
          (Transformation|AttachComponentToComponent _pickupsound (Variables|/Script/Engine.Actor|GetRootComponent PickupTarget) "root" "SnapToTarget" "SnapToTarget" "SnapToTarget" true))))))

(fn GetWadSizeValue
  (if (>= (Variables|Default|GetWadCount) (Math|Float|Truncate (ScalableFloat|GetValueatLevel (Variables|Default|GetLargeThreshold))))
    ;[1 1 1] Large
    (Variables|Default|SetWadSizeValue 2)
    (elif (>= (Variables|Default|GetWadCount) (Math|Float|Truncate (ScalableFloat|GetValueatLevel (Variables|Default|GetMediumThreshold))))
      ;[1 1 1] Medium
      (Variables|Default|SetWadSizeValue 1)
      (else
        ;[1 1 1] Small
        (Variables|Default|SetWadSizeValue 0))))
  (return (Variables|Default|GetWadSizeValue)))

(fn SetupPickupGlow
  (params (Level :type Integer))
  (bind _sm_rare_pickup (Variables|Default|Getsm_rare_pickup))
  (Rendering|Material|SetVectorParameterValueonMaterials _sm_rare_pickup "Color" (* "1.000000,0.400000, 0" (Utilities|Select :Index Level :Option%200 0.6 :Option%201 1.0 :Option%202 2.0)))
  (Rendering|Material|SetScalarParameterValueonMaterials _sm_rare_pickup "Expansion Lerp" (Utilities|Select :Index Level :Option%200 1.0 :Option%201 0.4 :Option%202 0.0))
  (Rendering|Material|SetScalarParameterValueonMaterials _sm_rare_pickup "Sprite Size" (Utilities|Select :Index Level :Option%200 4e+01 :Option%201 7e+01 :Option%202 1e+02)))

(fn ConstructionScript
  (bind _sm_rare_pickup (Variables|Default|Getsm_rare_pickup))
  (|CallParentFunction)
  (CallFunction|SetupPickupGlow 0)
  (Rendering|Material|SetScalarParameterValueonMaterials _sm_rare_pickup "Start Fade" 3e+03)
  (Rendering|Material|SetScalarParameterValueonMaterials _sm_rare_pickup "End Fade" 4e+03))
